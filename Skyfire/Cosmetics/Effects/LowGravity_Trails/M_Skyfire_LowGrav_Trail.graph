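# Translucent, two-sided, unlit parent material for Skyfire low-gravity cosmetic trails, built for Niagara ribbons and other particle types.
# Niagara dynamic parameters drive UV scale and offset, optional macro/micro noise distortion, and alpha erosion.
# Opacity combines texture and erosion noise with depth, Fresnel, camera, view-angle and UV-edge fades.
# Color blends texture-mapped tints or gradients, with optional chromatic aberration, HDR remapping and eye-adaptation exposure compensation on emissive.
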Material /Skyfire/Cosmetics/Effects/LowGravity_Trails/M_Skyfire_LowGrav_Trail
graph {
 properties {
  bUseMaterialAttributes=true
  bUsedWithBeamTrails=true
  bUsedWithMeshDeformer=true
  bUsedWithMeshParticles=true
  bUsedWithNiagaraMeshParticles=true
  bUsedWithNiagaraRibbons=true
  bUsedWithNiagaraSprites=true
  bUsedWithParticleSprites=true
  bUsedWithSkeletalMesh=true
  bUsedWithSplineMeshes=true
  blendMode=BLEND_Translucent
  pixelDepthOffsetMode=PDOM_Legacy
  shadingModel=MSM_Unlit
  twoSided=true
  usedShadingModels=Unlit
 }
 nodes [
  n1 = DynamicParameter() {
   parameterIndex=2
   defaultValue=<rgba 1 0 1 0>
   paramNames [Macro Distortion Scale U, Macro Distortion Offset U, Micro Distortion Scale U, Micro Distortion Offset U]
  }
  Add(
   A=Multiply(
    A=@n1.Macro Distortion Offset U
    B=ScalarParameter() {defaultValue=1, group=Macro_Distortion, parameterName=Distortion_Macro_Panspeed_X, sortPriority=1}
   )
   B=ScalarParameter() {group="5) Macro UV Distortion", parameterName=Macro_Distortion_Offset_X, sortPriority=6}
  )
  n2 = DynamicParameter() {defaultValue=<rgba 1 0 1 0>, paramNames [Color Scale U, Color Offset U, Base Scale U, Base Offset U]}
  Add(A=@n2.Base Offset U, B=ScalarParameter() {defaultValue=1, group=BaseTexture, parameterName=BaseTexture_Offset_U, sortPriority=1})
  n3 = ScalarParameter() {defaultValue=1, group="2) Opacity", parameterName=Base_Texture_Tile_U, sortPriority=22}
  n4 = ScalarParameter() {defaultValue=1, group="2) Opacity", parameterName=Base_Texture_Tile_V, sortPriority=23}
  n5 = LinearInterpolate(
   A=TextureCoordinate()
   B=TextureCoordinate() {coordinateIndex=1}
   Alpha=ScalarParameter() {desc=Texture Coordinate Channel to use for Macro UVs, group="2) Opacity", parameterName=Opacity UV Channel, sortPriority=21}
  )
  n6 = LinearInterpolate(
   A=@n5
   B=MaterialFunctionCall(XY=@n5) {materialFunction=&Swizzle}.YX
   Alpha=ScalarParameter() {group=Options, parameterName=Swizzle}
  )
  n7 = StaticSwitchParameter(True=Multiply(A=@n6, B=AppendVector(A=@n3, B=@n4)), False=@n6) {
   desc="This toggles on the legacy workflow and enables the material to be tuned from the material instance and the niagara "
   group=Static Switch
   parameterName=Enable Legacy Parameters
  }
  n8 = MaterialFunctionCall(Float2=@n7) {materialFunction=&BreakOutFloat2Components}
  n9 = AppendVector(A=Add(A=Multiply(A=@n8.R, B=@n2.Base Scale U), B=@n2.Base Offset U), B=@n8.G)
  n10 = ScalarParameter() {defaultValue=1, group=Macro_Distortion, parameterName=Distortion_Macro_Tile_X, sortPriority=3}
  n11 = ScalarParameter() {defaultValue=1, group=Macro_Distortion, parameterName=Distortion_Macro_Tile_Y, sortPriority=4}
  n12 = ScalarParameter() {desc=Texture Coordinate Channel to use for Macro UVs, group="5) Macro UV Distortion", parameterName=Macro UV Channel, sortPriority=1}
  n13 = LinearInterpolate(A=TextureCoordinate(), B=TextureCoordinate() {coordinateIndex=1}, Alpha=@n12)
  n14 = LinearInterpolate(
   A=@n13
   B=MaterialFunctionCall(XY=@n13) {materialFunction=&Swizzle}.YX
   Alpha=ScalarParameter() {group=Options, parameterName=Swizzle}
  )
  n15 = StaticSwitchParameter(True=Multiply(A=@n14, B=AppendVector(A=@n10, B=@n11)), False=@n14) {
   desc="This toggles on the legacy workflow and enables the material to be tuned from the material instance and the niagara "
   group=Static Switch
   parameterName=Enable Legacy Parameters
  }
  n16 = MaterialFunctionCall(Float2=@n15) {materialFunction=&BreakOutFloat2Components}
  n17 = AppendVector(A=Add(A=Multiply(A=@n16.R, B=@n1.Macro Distortion Scale U), B=@n1.Macro Distortion Offset U), B=@n16.G)
  n18 = MaterialFunctionCall(XY=@n17) {materialFunction=&Swizzle}
  n19 = ScalarParameter() {group="7.3) Macro Distortion Texture - Rotate UVs", parameterName=Macro Distortion Texture - Rotate UVs, sortPriority=49}
  n20 = Constant() {r=1}
  n21 = Constant() {r=-1}
  n22 = ScalarParameter() {group="7.3) Macro Distortion Texture - Rotate UVs", parameterName=Macro Distortion Texture - Flip UVs Horizontal, sortPriority=50}
  n23 = LinearInterpolate(A=@n20, B=@n21, Alpha=@n22)
  n24 = ScalarParameter() {group="7.3) Macro Distortion Texture - Rotate UVs", parameterName=Macro Distortion Texture - Flip UVs Vertically, sortPriority=51}
  n25 = LinearInterpolate(A=@n20, B=@n21, Alpha=@n24)
  n26 = TextureSampleParameter2D(UVs=Multiply(A=LinearInterpolate(A=@n17, B=@n18.YX, Alpha=@n19), B=AppendVector(A=@n23, B=@n25))) {
   desc=Texture for large uv distortions.
   group="5) Macro UV Distortion"
   parameterName=Macro Distortion Texture
   samplerSource=SSM_Wrap_WorldGroupSettings
   sortPriority=0
   texture=&AtmosphericCloudNoise01
  }
  n27 = Constant()
  n28 = Add(A=Multiply(A=AppendVector(A=@n26.R, B=@n27)) {constB=2}) {constB=-1}
  n29 = Constant()
  n30 = Add(A=Multiply(A=AppendVector(A=@n29, B=@n26.R)) {constB=2}) {constB=-1}
  n31 = ScalarParameter() {
   defaultValue=1
   desc=Lerp between the different directions when is Grey Scale is 1
   group="5) Macro UV Distortion"
   parameterName="Direction (U or V)"
   sortPriority=8
  }
  n32 = ScalarParameter() {defaultValue=1, desc=Set to 1 if your texture is mono color., group="5) Macro UV Distortion", parameterName=is GreyScale, sortPriority=2}
  n33 = Add(
   A=LinearInterpolate(A=AppendVector(A=@n26.R, B=@n26.R), B=LinearInterpolate(A=@n28, B=@n30, Alpha=@n31), Alpha=@n32)
   B=AppendVector(
    A=ScalarParameter() {desc=Offset macro distortion U, group="5) Macro UV Distortion", parameterName=Macro Distortion U Offset, sortPriority=6}
    B=ScalarParameter() {desc=Offset macro distortion V, group="5) Macro UV Distortion", parameterName=Macro Distortion V Offset, sortPriority=7}
   )
  ) {constB=0}
  n34 = DynamicParameter() {parameterIndex=3, defaultValue=<rgba 0 0 1 0>, paramNames [Macro Distortion Amount, Macro Distortion Offset U, Alpha Power, Alpha Erosion]}
  n35 = DynamicParameter() {
   parameterIndex=2
   defaultValue=<rgba 1 0 1 0>
   paramNames [Macro Distortion Scale U, Macro Distortion Offset U, Micro Distortion Scale U, Micro Distortion Offset U]
  }
  n36 = ScalarParameter() {defaultValue=1, group=Micro_Distortion, parameterName=Distortion_Micro_Tile_X, sortPriority=3}
  n37 = ScalarParameter() {defaultValue=1, group=Micro_Distortion, parameterName=Distortion_Micro_Tile_Y, sortPriority=4}
  n38 = ScalarParameter() {desc=Texture Coordinate Channel to use for Micro UVs, group="6) Micro UV Distortion", parameterName=Micro UV Channel, sortPriority=1}
  n39 = LinearInterpolate(A=TextureCoordinate(), B=TextureCoordinate() {coordinateIndex=1}, Alpha=@n38)
  n40 = StaticSwitchParameter(True=Multiply(A=@n39, B=AppendVector(A=@n36, B=@n37)), False=@n39) {
   desc="This toggles on the legacy workflow and enables the material to be tuned from the material instance and the niagara  THis allows the shader's parameters to be tweaked from two different places."
   group=Static Switch
   parameterName=Enable Legacy Parameters
  }
  n41 = MaterialFunctionCall(Float2=@n40) {materialFunction=&BreakOutFloat2Components}
  n42 = AppendVector(A=Add(A=Multiply(A=@n41.R, B=@n35.Micro Distortion Scale U), B=@n35.Micro Distortion Offset U), B=@n41.G)
  n43 = MaterialFunctionCall(XY=@n42) {materialFunction=&Swizzle}
  n44 = ScalarParameter() {group="7.4) Micro Distortion Texture Rotation", parameterName=Micro Distortion Texture - Rotate UVs, sortPriority=52}
  n45 = Constant() {r=1}
  n46 = Constant() {r=-1}
  n47 = LinearInterpolate(
   A=@n45
   B=@n46
   Alpha=ScalarParameter() {group="7.4) Micro Distortion Texture Rotation", parameterName=Micro Distortion Texture - Flip UVs Horizontal, sortPriority=53}
  )
  n48 = LinearInterpolate(
   A=@n45
   B=@n46
   Alpha=ScalarParameter() {group="7.4) Micro Distortion Texture Rotation", parameterName=Micro Distortion Texture - Flip UVs Vertically, sortPriority=54}
  )
  n49 = TextureSampleParameter2D(UVs=Multiply(A=LinearInterpolate(A=@n42, B=@n43.YX, Alpha=@n44), B=AppendVector(A=@n47, B=@n48))) {
   desc=Texture for small uv distortion details
   group="6) Micro UV Distortion"
   parameterName=Micro Distortion Texture
   samplerSource=SSM_Wrap_WorldGroupSettings
   sortPriority=0
   texture=/Game/Global/Textures/Default/Noise/T_Dudebro_Noise_01
  }
  n50 = Add(A=Multiply(A=AppendVector(A=@n49.R, B=@n49.R)) {constB=2}) {constB=-1}
  n51 = DynamicParameter() {parameterIndex=3, defaultValue=<rgba 0 0 1 0>, paramNames [Macro Distortion Amount, Macro Distortion Offset U, Alpha Power, Alpha Erosion]}
  n52 = ShadingPathSwitch(Default=Add(A=Multiply(A=@n33, B=@n34.Macro Distortion Amount) {constB=0.05}, B=Multiply(A=@n50, B=@n51.Macro Distortion Offset U) {constB=0.05}))
  n53 = MaterialFunctionCall(Float2=StaticSwitchParameter(True=Add(A=@n9, B=@n52) {desc=Add uv distortion}, False=@n9) {desc=Only one child material instance should have this enabled., group=Static Switch, parameterName=Use Distortion}) {materialFunction=&BreakOutFloat2Components}
  n54 = AppendVector(
   A=LinearInterpolate(
    A=@n53.R
    B=Clamp(None=@n53.R)
    Alpha=ScalarParameter() {desc=Clamps the opacity's uvs., group="2) Opacity", parameterName=Clamp U, sortPriority=24}
   )
   B=LinearInterpolate(
    A=@n53.G
    B=Clamp(None=@n53.G)
    Alpha=ScalarParameter() {defaultValue=1, desc=Clamps the opacity's uvs., group="2) Opacity", parameterName=Clamp V, sortPriority=25}
   )
  )
  n55 = MaterialFunctionCall(XY=@n54) {materialFunction=&Swizzle}
  n56 = ScalarParameter() {group="7) Opacity Texture Rotation", parameterName=Opacity Texture - Rotate UVs, sortPriority=40}
  n57 = Constant() {r=1}
  n58 = Constant() {r=-1}
  n59 = LinearInterpolate(
   A=@n57
   B=@n58
   Alpha=ScalarParameter() {group="7) Opacity Texture Rotation", parameterName=Opacity Texture - Flip UVs Horizontal, sortPriority=41}
  )
  n60 = LinearInterpolate(
   A=@n57
   B=@n58
   Alpha=ScalarParameter() {group="7) Opacity Texture Rotation", parameterName=Opacity Texture - Flip UVs Vertically, sortPriority=42}
  )
  n61 = TextureSampleParameter2D(UVs=Multiply(A=LinearInterpolate(A=@n54, B=@n55.YX, Alpha=@n56), B=AppendVector(A=@n59, B=@n60))) {
   desc=Texture for the opacity. Controls are available in the FN Ribbon Material module.
   group="2) Opacity"
   parameterName=Opacity Texture
   samplerSource=SSM_Wrap_WorldGroupSettings
   sortPriority=20
   texture=/Skyfire/Athena/Effects/Particles/Skydiving/NS_Hightower_Date/T_Hightower_Ribbon_Macro
  }
  n62 = DynamicParameter() {parameterIndex=3, defaultValue=<rgba 0 0 1 0>, paramNames [Macro Distortion Amount, Macro Distortion Offset U, Alpha Power, Alpha Erosion]}
  n63 = Multiply(A=Power(Base=@n61.RGB, Exp=@n62.Alpha Power), B=ParticleColor().A)
  n64 = DynamicParameter() {parameterIndex=3, defaultValue=<rgba 0 0 1 0>, paramNames [Macro Distortion Amount, Macro Distortion Offset U, Alpha Power, Alpha Erosion]}
  n65 = DynamicParameter() {parameterIndex=1, defaultValue=<rgba 1 0 1 0>, paramNames [Erosion Scale U, Erosion Offset U, Erosion Power, Erosion Hardness]}
  n66 = LinearInterpolate(
   A=TextureCoordinate()
   B=TextureCoordinate() {coordinateIndex=1}
   Alpha=ScalarParameter() {desc=Texture Coordinate Channel to use for Erosion UVs, group="3) Erosion Mask", parameterName=Erosion UV Channel}
  )
  n67 = MaterialFunctionCall(Float2=LinearInterpolate(
    A=@n66
    B=MaterialFunctionCall(XY=@n66) {materialFunction=&Swizzle}.YX
    Alpha=ScalarParameter() {group=Options, parameterName=Swizzle}
   )) {materialFunction=&BreakOutFloat2Components}
  n68 = AppendVector(A=Add(A=Multiply(A=@n67.R, B=@n65.Erosion Scale U), B=@n65.Erosion Offset U), B=@n67.G)
  n69 = MaterialFunctionCall(XY=@n68) {materialFunction=&Swizzle}
  n70 = ScalarParameter() {group="7.2) Erosion Texture Rotation", parameterName=Erosion Texture - Rotate UVs, sortPriority=46}
  n71 = Constant() {r=1}
  n72 = Constant() {r=-1}
  n73 = LinearInterpolate(
   A=@n71
   B=@n72
   Alpha=ScalarParameter() {group="7.2) Erosion Texture Rotation", parameterName=Erosion Texture - Flip UVs Horizontal, sortPriority=47}
  )
  n74 = LinearInterpolate(
   A=@n71
   B=@n72
   Alpha=ScalarParameter() {group="7.2) Erosion Texture Rotation", parameterName=Erosion Texture - Flip UVs Vertically, sortPriority=48}
  )
  n75 = TextureSampleParameter2D(UVs=Multiply(A=LinearInterpolate(A=@n68, B=@n69.YX, Alpha=@n70), B=AppendVector(A=@n73, B=@n74))) {
   desc=Texture for the erosion effect. This multiplies against the opactiy texture. Controls are available in the FN Ribbon Material module.
   group="3) Erosion Mask"
   parameterName=ErosionNoise
   samplerSource=SSM_Wrap_WorldGroupSettings
   samplerType=SAMPLERTYPE_Masks
   sortPriority=31
   texture=/Game/Global/Textures/Default/Noise/T_CellularNoise
  }
  n76 = DynamicParameter() {parameterIndex=1, defaultValue=<rgba 1 0 1 0>, paramNames [Erosion Scale U, Erosion Offset U, Erosion Power, Erosion Hardness]}
  n77 = Add(A=Power(Base=@n75.RGB, Exp=@n76.Erosion Power))
  n78 = Multiply(A=Max(A=Subtract(A=@n63, B=@n64.Alpha Erosion)) {constB=0}, B=Clamp(None=Subtract(A=@n77, B=@n76.Erosion Hardness)))
  n79 = Multiply(
   A=@n78
   B=DepthFade(FadeDistance=ScalarParameter() {defaultValue=250, desc=Depth fade distance., group="4) Fade", parameterName=Depth Fade Distance})
  )
  n80 = DotProduct(A=CameraVectorWS(), B=VertexNormalWS())
  n81 = Multiply(
   A=ShadingPathSwitch(Default=@n79, Deferred=@n79, Forward=@n79, Mobile=@n78)
   B=Power(
    Base=Abs(None=@n80)
    Exp=ScalarParameter() {defaultValue=1, desc=Exponent for frensel falloff., group="4) Fade", parameterName=Frenel Fade Exponent}
   )
  )
  n82 = Power(
   Base=MaterialFunctionCall(
    Fade Length=ScalarParameter() {desc=Camera Fade distance., group="4) Camera Fade", parameterName=Camera Fade Distance, sortPriority=4}
    Fade Offset=ScalarParameter() {desc=Camera fade offset., group="4) Camera Fade", parameterName=Camera Fade Offset, sortPriority=5}
   ) {materialFunction=&CameraDepthFade}.Result
   Exp=ScalarParameter() {defaultValue=1, desc=Camera fade exponent., group="4) Camera Fade", parameterName=Camera Fade Exponent, sortPriority=6}
  )
  n83 = ScalarParameter() {group="4.1) View Angle Fade", parameterName=Fade Sensitivity, sortPriority=0}
  n84 = ScalarParameter() {
   defaultValue=0.05
   desc=Length in uv space of the leading edge fading in.
   group="4) UV Fade"
   parameterName=Leading Edge Fade Distance
   sortPriority=0
  }
  n85 = Max(A=@n84) {constB=0.0001}
  n86 = LinearInterpolate(A=TextureCoordinate(), B=TextureCoordinate() {coordinateIndex=1}, Alpha=ScalarParameter() {parameterName=Edge Gradient UV Channel})
  n87 = ComponentMask(None=@n86) {r=true}
  n88 = Clamp(None=Multiply(A=@n87, B=Divide(B=@n85)))
  n89 = ScalarParameter() {defaultValue=2.2, desc=Exponent of the leading edge fade in., group="4) UV Fade", parameterName=Leading Edge Fade Power, sortPriority=1}
  n90 = OneMinus(None=@n87)
  n91 = ScalarParameter() {
   defaultValue=0.1
   desc=Length in uv space of the trailing edge fading out.
   group="4) UV Fade"
   parameterName=Trailing Edge Fade Distance
   sortPriority=2
  }
  n92 = Multiply(A=@n90, B=Divide(B=Max(A=@n91) {constB=0.0001}))
  n93 = ScalarParameter() {defaultValue=2.2, desc=Exponent of the trailing edge fade out., group="4) UV Fade", parameterName=Trailing Edge Fade Power, sortPriority=3}
  n94 = Clamp(None=Power(Base=Clamp(None=@n92), Exp=@n93) {constExponent=2.2})
  n95 = Min(A=Clamp(None=Power(Base=@n88, Exp=@n89) {constExponent=2.2}), B=@n94)
  n96 = ScalarParameter() {
   defaultValue=0.05
   desc=Length in uv space of the leading edge fading in.
   group="4) UV Fade"
   parameterName=Top Edge Fade Distance
   sortPriority=0
  }
  n97 = Max(A=@n96) {constB=0.0001}
  n98 = ComponentMask(None=@n86) {g=true}
  n99 = Clamp(None=Multiply(A=@n98, B=Divide(B=@n97)))
  n100 = ScalarParameter() {defaultValue=2.2, desc=Exponent of the leading edge fade in., group="4) UV Fade", parameterName=Top Edge Fade Power, sortPriority=1}
  n101 = ScalarParameter() {
   defaultValue=0.1
   desc=Length in uv space of the trailing edge fading out.
   group="4) UV Fade"
   parameterName=Bottom Edge Fade Distance
   sortPriority=2
  }
  n102 = Max(A=@n101) {constB=0.0001}
  n103 = Clamp(None=Multiply(A=OneMinus(None=@n98), B=Divide(B=@n102)))
  n104 = ScalarParameter() {defaultValue=2.2, desc=Exponent of the trailing edge fade out., group="4) UV Fade", parameterName=Bottom Edge Fade Power, sortPriority=3}
  n105 = Multiply(A=Clamp(None=Power(Base=@n99, Exp=@n100) {constExponent=2.2}), B=Clamp(None=Power(Base=@n103, Exp=@n104) {constExponent=2.2}))
  n106 = Multiply(
   A=Multiply(
    A=Multiply(A=@n81, B=@n82)
    B=MaterialFunctionCall("FadeSensitivity (optional)"=@n83) {materialFunction=&FadeBasedOnViewAngle}.Result
   )
   B=LinearInterpolate(A=@n95, B=Multiply(A=@n105, B=@n95), Alpha=ScalarParameter() {parameterName=Use Top and Bottom Opacity Gradient 0-1})
  )
  n107 = TextureCoordinate()
  n108 = Multiply(A=Subtract(A=ComponentMask(None=@n107) {desc=UV0 G seems broke., g=true}, B=Constant() {r=0.5}), B=Constant() {r=2})
  n109 = Multiply(A=@n108, B=@n108)
  n110 = ComponentMask(None=@n90) {desc=UV0 G seems broke., r=true}
  n111 = ScalarParameter() {
   defaultValue=0.5
   desc=Amount of rounding on the base of the ribbon
   group="6.1) Opacity"
   parameterName=Ribbon Rounding
   sliderMax=1
   sortPriority=1
  }
  n112 = OneMinus(None=Saturate(None=Divide(A=@n110, B=@n111)))
  n113 = Subtract(A=OneMinus(None=Saturate(None=@n109)), B=Multiply(A=@n112, B=@n112))
  n114 = ScalarParameter() {defaultValue=0.2, desc=Softness, group="6.1) Opacity", parameterName=Ribbon Rounding Softness, sliderMax=1, sortPriority=2}
  n115 = OneMinus(None=Abs(None=@n108))
  n116 = LinearInterpolate(A=@n114, Alpha=Multiply(A=@n115, B=@n115))
  Multiply(A=Multiply(A=@n106, B=@n94), B=Saturate(None=Divide(A=@n113, B=@n116)))
  n117 = DynamicParameter() {desc=Use FN Ribbon Material Module, defaultValue=<rgba 1 0 1 0>, paramNames [Color Scale U, Color Offset U, Base Scale U, Base Offset U]}
  Multiply(A=@n117.Color Offset U, B=ScalarParameter() {defaultValue=1, group=Color_Texture, parameterName=Color_Texture_PanSpeed_X, sortPriority=1})
  Multiply(
   A=@n35.Micro Distortion Offset U
   B=ScalarParameter() {defaultValue=1, group=Micro_Distortion, parameterName=Distortion_Micro_Panspeed_X, sortPriority=1}
  )
  n118 = SceneColor!(OffsetFraction=VectorParameter() {parameterName=Red, defaultValue=<rgba 1 0 0 0>}.RGB) {inputMode=OffsetFraction}
  n119 = SceneColor(OffsetFraction=VectorParameter() {parameterName=Green, defaultValue=<rgba 0 1 0 0>}.RGB) {inputMode=OffsetFraction}
  n120 = SceneColor(OffsetFraction=VectorParameter() {parameterName=Blue, defaultValue=<rgba 0 0 1 0>}.RGB) {inputMode=OffsetFraction}
  Multiply(
   A=MaterialFunctionCall(
    X=MaterialFunctionCall(Float3=@n118.RGB) {materialFunction=&BreakOutFloat3Components}.R
    Y=MaterialFunctionCall(Float3=@n119.RGB) {materialFunction=&BreakOutFloat3Components}.G
    Z=MaterialFunctionCall(Float3=@n120.RGB) {materialFunction=&BreakOutFloat3Components}.B
   ) {materialFunction=&MakeFloat3}.Result
   B=ScalarParameter() {defaultValue=20, parameterName=Chromatic Strength}
  )
  Multiply(A=@n2.Base Offset U, B=ScalarParameter() {defaultValue=1, group=BaseTexture, parameterName=BaseTexture_PanSpeed_U, sortPriority=1})
  Multiply(A=@n50, B=ScalarParameter() {defaultValue=0.0125, group=Micro_Distortion, parameterName=Micro_Distortion_Intensity, sortPriority=5})
  OneMinus(None=@n95)
  n121 = VectorParameter() {
   desc=Maps to the black values of the color texture.
   group="1) Color"
   parameterName=Texture_Color_A
   sortPriority=1
   defaultValue=<rgba 0.8078 1 0.1412 1>
  }
  n122 = ScalarParameter() {defaultValue=1, group="1) Color", parameterName=Color_Texture_Tile_U, sortPriority=6}
  n123 = ScalarParameter() {defaultValue=1, group="1) Color", parameterName=Color_Texture_Tile_V, sortPriority=7}
  n124 = LinearInterpolate(
   A=TextureCoordinate()
   B=TextureCoordinate() {coordinateIndex=1}
   Alpha=ScalarParameter() {desc=Set the Texture Coordinate channel to use., group="1) Color", parameterName=Color UV Channel, sortPriority=4}
  )
  n125 = LinearInterpolate(
   A=@n124
   B=MaterialFunctionCall(XY=@n124) {materialFunction=&Swizzle}.YX
   Alpha=ScalarParameter() {group=Options, parameterName=Swizzle}
  )
  n126 = StaticSwitchParameter(True=Multiply(A=@n125, B=AppendVector(A=@n122, B=@n123)), False=@n125) {
   desc="This toggles on the legacy workflow and enables the material to be tuned from the material instance and the niagara "
   group=Static Switch
   parameterName=Enable Legacy Parameters
  }
  n127 = MaterialFunctionCall(Float2=@n126) {materialFunction=&BreakOutFloat2Components}
  n128 = AppendVector(
   A=Add(
    A=ScalarParameter() {defaultValue=1, group="1.2) Color Texture Chromatic Abberation", parameterName=Color Abberation 01 Offset U, sortPriority=1}
    B=@n127.R
   )
   B=Add(
    A=ScalarParameter() {defaultValue=1, group="1.2) Color Texture Chromatic Abberation", parameterName=Color Abberation 01 Offset V, sortPriority=2}
    B=@n127.G
   )
  )
  n129 = MaterialFunctionCall(XY=@n128) {materialFunction=&Swizzle}
  n130 = ScalarParameter() {group="7.1) Color Texture Rotation", parameterName=Color Texture - Rotate UVs, sortPriority=43}
  n131 = Constant() {r=1}
  n132 = Constant() {r=-1}
  n133 = LinearInterpolate(
   A=@n131
   B=@n132
   Alpha=ScalarParameter() {group="7.1) Color Texture Rotation", parameterName=Color Texture - Flip UVs Horizontal, sortPriority=44}
  )
  n134 = LinearInterpolate(
   A=@n131
   B=@n132
   Alpha=ScalarParameter() {group="7.1) Color Texture Rotation", parameterName=Color Texture - Flip UVs Vertically, sortPriority=45}
  )
  n135 = TextureSampleParameter2D(UVs=Multiply(A=LinearInterpolate(A=@n128, B=@n129.YX, Alpha=@n130), B=AppendVector(A=@n133, B=@n134))) {
   desc=Texture to break up the color of the ribbon.
   group="1) Color"
   parameterName=Color Texture
   samplerSource=SSM_Wrap_WorldGroupSettings
   sortPriority=3
   texture=/Skyfire/Athena/Effects/Particles/Skydiving/NS_Hightower_Date/T_Hightower_Ribbon_Color
  }
  n136 = StaticSwitchParameter(
   True=Add(A=ScalarParameter() {group="1) Color", parameterName=Color_Texture_Offset_X, sortPriority=6}, B=@n117.Color Offset U)
   False=@n117.Color Offset U
  ) {
   desc="This toggles on the legacy workflow and enables the material to be tuned from the material instance and the niagara "
   group=Static Switch
   parameterName=Enable Legacy Parameters
  }
  n137 = AppendVector(A=Add(A=Multiply(A=@n127.R, B=@n117.Color Scale U), B=@n136), B=@n127.G)
  n138 = StaticSwitchParameter(True=Add(A=@n137, B=@n52) {desc=Add uv distortion}, False=@n137) {desc=Only one child material instance should have this enabled., group=Static Switch, parameterName=Use Distortion}
  n139 = MaterialFunctionCall(XY=@n138) {materialFunction=&Swizzle}
  n140 = ScalarParameter() {group="7.1) Color Texture Rotation", parameterName=Color Texture - Rotate UVs, sortPriority=43}
  n141 = Constant() {r=1}
  n142 = Constant() {r=-1}
  n143 = LinearInterpolate(
   A=@n141
   B=@n142
   Alpha=ScalarParameter() {group="7.1) Color Texture Rotation", parameterName=Color Texture - Flip UVs Horizontal, sortPriority=44}
  )
  n144 = LinearInterpolate(
   A=@n141
   B=@n142
   Alpha=ScalarParameter() {group="7.1) Color Texture Rotation", parameterName=Color Texture - Flip UVs Vertically, sortPriority=45}
  )
  n145 = TextureSampleParameter2D(UVs=Multiply(A=LinearInterpolate(A=@n138, B=@n139.YX, Alpha=@n140), B=AppendVector(A=@n143, B=@n144))) {
   desc=Texture to break up the color of the ribbon.
   group="1) Color"
   parameterName=Color Texture
   samplerSource=SSM_Wrap_WorldGroupSettings
   sortPriority=3
   texture=&T_Hightower_Ribbon_Color
  }
  n146 = AppendVector(
   A=Add(
    A=ScalarParameter() {defaultValue=1, group="1.2) Color Texture Chromatic Abberation", parameterName=Color Abberation 02 Offset U, sortPriority=3}
    B=@n127.R
   )
   B=Add(
    A=ScalarParameter() {defaultValue=1, group="1.2) Color Texture Chromatic Abberation", parameterName=Color Abberation 02 Offset V, sortPriority=4}
    B=@n127.G
   )
  )
  n147 = MaterialFunctionCall(XY=@n146) {materialFunction=&Swizzle}
  n148 = ScalarParameter() {group="7.1) Color Texture Rotation", parameterName=Color Texture - Rotate UVs, sortPriority=43}
  n149 = Constant() {r=1}
  n150 = Constant() {r=-1}
  n151 = LinearInterpolate(
   A=@n149
   B=@n150
   Alpha=ScalarParameter() {group="7.1) Color Texture Rotation", parameterName=Color Texture - Flip UVs Horizontal, sortPriority=44}
  )
  n152 = LinearInterpolate(
   A=@n149
   B=@n150
   Alpha=ScalarParameter() {group="7.1) Color Texture Rotation", parameterName=Color Texture - Flip UVs Vertically, sortPriority=45}
  )
  n153 = TextureSampleParameter2D(UVs=Multiply(A=LinearInterpolate(A=@n146, B=@n147.YX, Alpha=@n148), B=AppendVector(A=@n151, B=@n152))) {
   desc=Texture to break up the color of the ribbon.
   group="1) Color"
   parameterName=Color Texture
   samplerSource=SSM_Wrap_WorldGroupSettings
   sortPriority=3
   texture=&T_Hightower_Ribbon_Color
  }
  n154 = VectorParameter() {parameterName=Trailing Gradient Color, defaultValue=<rgba 0.2831 1 0.0595 1>}
  n155 = TextureCoordinate()
  n156 = Power(Base=ComponentMask(None=@n155) {r=true}, Exp=ScalarParameter() {defaultValue=2, parameterName=Leading Color Gradient Power})
  n157 = VectorParameter() {parameterName=Leading Gradient Color, defaultValue=<rgba 0.8078 1 0.1412 1>}
  n158 = Power(Base=ComponentMask(None=OneMinus(None=@n155)) {r=true}, Exp=ScalarParameter() {defaultValue=2, parameterName=Trailing Color Gradient Power})
  n159 = VectorParameter() {
   desc=Maps to the white values of the Color Texture.
   group="1) Color"
   parameterName=Texture_Color_B
   sortPriority=2
   defaultValue=<rgba 0.2831 1 0.0595 1>
  }
  n160 = StaticSwitchParameter(True=LinearInterpolate(A=Multiply(A=@n154.RGB, B=@n156), B=Multiply(A=@n157.RGB, B=@n158), Alpha=Constant() {r=0.5}), False=@n159.RGB) {parameterName=Use Gradient for Color B}
  n161 = Add(A=AppendVector(A=AppendVector(A=@n135.R, B=@n145.G), B=@n153.B), B=@n160)
  n162 = Add(A=@n135.RGBA, B=@n153.RGBA)
  n163 = StaticComponentMaskParameter(None=StaticSwitchParameter(True=Add(A=@n162, B=@n145.RGBA), False=@n145.RGBA) {group="1.2) Color Texture Chromatic Abberation", parameterName=Use Abberation Setup?, sortPriority=0}) {defaultB=true, defaultG=true, defaultR=true, group="1.1) Color Correction", parameterName=Color Texture Channel, sortPriority=0}
  n164 = ScalarParameter() {group="1.1) Color Correction", parameterName=Color Desaturation, sortPriority=1}
  n165 = ScalarParameter() {defaultValue=2, desc=Multiplies the color texture after the power., group="1.1) Color Correction", parameterName=Color Scale, sortPriority=6}
  n166 = Multiply(
   A=Power(
    Base=Desaturation(None=@n163, Fraction=@n164)
    Exp=ScalarParameter() {defaultValue=1, desc=Power of color texture., group="1.1) Color Correction", parameterName=Color Power, sortPriority=5}
   )
   B=@n165
  )
  n167 = ScalarParameter() {defaultValue=1, desc="If disabled, color defaults to Particle Color.", group="1) Color", parameterName=Use Color Texture, sortPriority=0}
  n168 = LinearInterpolate(
   A=ParticleColor().RGB
   B=LinearInterpolate(
    A=@n121.RGB
    B=StaticSwitchParameter(True=@n161, False=@n160) {group="1.2) Color Texture Chromatic Abberation", parameterName=Use Abberation Setup?, sortPriority=0}
    Alpha=Clamp(None=@n166)
   )
   Alpha=@n167
  )
  n169 = Saturate(None=@n168)
  n170 = ParticleColor()
  n171 = LinearInterpolate(
   A=@n168
   B=MaterialFunctionCall(
    LDR Input=@n169
    HDR Tint=Multiply(A=@n169, B=@n170.RGB)
    HDR Intensity Multiplier=ScalarParameter() {defaultValue=1, parameterName=HDR Multiplier}
   ) {materialFunction=&DeriveHDRfromLDR}.RemappedRange
   Alpha=ScalarParameter() {defaultValue=1, parameterName=UseHDR}
  )
  n172 = Multiply(A=@n106, B=ParticleColor().A)
  MaterialExpressionGetMaterialAttributes_0 = GetMaterialAttributes(None=SetMaterialAttributes!(
    MaterialAttributes=SetMaterialAttributes!(MaterialAttributes, Emissive Color=@n171, Base Color=@n171) {attributeSetTypes [#1, #2]}
    Opacity=@n172
    Opacity Mask=@n172
   ) {attributeSetTypes [#3, #4]}) {attributeGetTypes [#1]}
  MaterialExpressionScalarParameter_7 = ScalarParameter() {group=Options, parameterName=Use Eye Adaption Exposure Control for Emissives}
  MaterialExpressionEyeAdaptation_0 = EyeAdaptation()
  n173 = Custom!(
   LightValue=@MaterialExpressionGetMaterialAttributes_0.Emissive Color
   Alpha=@MaterialExpressionScalarParameter_7
   Adaptation=@MaterialExpressionEyeAdaptation_0.EyeAdaptation
  ) {
   code="float LerpLogScale = -Alpha * log(Adaptation);\r\nfloat Scale = exp(LerpLogScale);\r\nreturn LightValue * Scale;\r\n\r\n\r\n"
   inputs [
    {
     inputName=LightValue
     input {expression=&self:MaterialExpressionGetMaterialAttributes_0, outputIndex=1}
    }
    {inputName=Alpha, input {expression=&self:MaterialExpressionScalarParameter_7}}
    {
     inputName=Adaptation
     input {expression=&self:MaterialExpressionEyeAdaptation_0}
    }
   ]
  }
  SetMaterialAttributes!(MaterialAttributes=@MaterialExpressionGetMaterialAttributes_0.MaterialAttributes, Emissive Color=@n173) {attributeSetTypes [#1]}
 ]
}
