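# Niagara cosmetic effect for Skyfire low-gravity trails on a skeletal mesh, combining looping sprites with an animation-driven ribbon.
# Decals_Loop continuously spawns user-colored sprites across 16 spawn groups, using each group as a bone/socket index to re-snap particles every update.
# Curves drive sprite alpha and dynamic material parameters over particle life.
# The AnimTrail emitter builds short-lived ribbons from bone indices, with panning, distorted, eroded master material that fades out.
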NiagaraSystem /Skyfire/Cosmetics/Effects/LowGravity_Trails/NS_Skyfire_LowGravity_Trails
dependencies {
 usedDataInterfaces [&NiagaraDataInterfaceCurve]
 usedDynamicInputs [&HueShiftLinearColor]
 usedModules [
  &SystemState
  &self:SystemSpawnScript.SystemScriptSource.SystemScriptGraph.NiagaraNodeAssignment_2.SetVariables_#1
  &self:SystemSpawnScript.SystemScriptSource.SystemScriptGraph.NiagaraNodeAssignment_0.SetVariables_#2
 ]
}
emitters {
 Decals_Loop {
  emitterUpdateScript [
   &EmitterState(Life Cycle Mode=::NewEnumerator1, Loop Duration=1, UseLoopDelay=false)
   &SpawnRate(SpawnRate=17, Use Spawn Probability=false)
  ]
  particleSpawnScript [
   &FN_SetSpawnGroup_v2(Number of Spawn Groups=16)
   /Niagara/Modules/Spawn/Initialization/InitializeParticle(
    Color=$User.Color
    Module.Write Color=true
    Module.Write Lifetime=true
    Module.Write Mass=true
    Module.Write Material Random=true
    Module.Write Position=false
    Module.Write Scale=false
    Module.Write SpriteRotation=false
    Module.Write SpriteSize=true
    Position=&SimulationPosition()
    Sprite Size=&Vector2DFromFloat(Value=30)
   )
   &DynamicMaterialParameters(
    Index 0 Param 2=1
    Index 0 Param 4=&UniformRangedFloat(Minimum=0, Maximum=1)
    Module.Param0WriteEnabled=false
    Module.Param1WriteEnabled=true
    Module.Param2WriteEnabled=false
    Module.Param3WriteEnabled=true
   )
   &CameraOffset(Camera Offset Amount=30)
  ]
  particleUpdateScript [
   &ParticleState()
   &SkeletalMeshLocation(
    Bone / Socket Index=$Particles.SpawnGroup
    Bone Sampling Mode=::NewEnumerator4
    Inherit Velocity=false
    MaskSpawnGroup=false
    Skeletal Mesh=$User.SkeletalMesh
   )
   &SolveForcesAndVelocity(Clamp Velocity=false, Limit Acceleration=false)
   &ScaleColor(
    Scale Alpha=&FloatFromCurve(FloatCurve=DataInterface() {propertyValues {Curve {keys [<at 0.5 1 RCIM_Cubic>]}}}, Scale Curve=2)
    Scale RGB=&Lerp_Vector(A=<1 0.6851 0.8647>, B=<0.3728 1 0.6644>, Alpha=$Particles.NormalizedAge)
    ScaleA=true
    ScaleRGB=false
   )
   &DynamicMaterialParameters(
    Index 0 Param 1=&FloatFromCurve(FloatCurve=DataInterface() {propertyValues {Curve {keys [
         {arriveTangent=-11.7199, interpMode=RCIM_Cubic, leaveTangent=-11.7199, tangentMode=RCTM_User, time=0, value=5}
         <at 0.4 1.5 RCIM_Cubic>
         {arriveTangent=7.1904, interpMode=RCIM_Cubic, leaveTangent=7.1904, tangentMode=RCTM_User, time=1, value=5}
        ]}}})
    Index 0 Param 3=&Lerp_Float(A=0.35, B=0, Alpha=$Particles.NormalizedAge)
    Module.Param0WriteEnabled=true
    Module.Param1WriteEnabled=false
    Module.Param2WriteEnabled=true
    Module.Param3WriteEnabled=false
   ) {module=DynamicMaterialParameters001}
  ]
  rendererClasses [&NiagaraSpriteRendererProperties]
  renderers [{rendererClass=&NiagaraSpriteRendererProperties}]
 }
 FN_AnimTrail_with_MasterMaterial {
  emitterUpdateScript [
   &EmitterState()
   &FN_SpawnAnimTrail_v2(Spawn Mode:ENiagaraSkeletalSpawning=::NewEnumerator0)
  ]
  particleSpawnScript [
   /Game/Effects/Niagara/Modules/AnimTrail/FN_InitializeAnimTrail/FN_InitializeAnimTrailBoneIndecies(Second Bone or Socket Index=1)
   /Game/Effects/Niagara/Modules/AnimTrail/FN_InitializeAnimTrail/FN_InitializeAnimTrail_v3/FN_InitializeAnimTrail(AnimTrailWidth=2.5, Lifetime=1)
  ]
  particleUpdateScript [
   &SolveForcesAndVelocity(Clamp Velocity=false, Limit Acceleration=false)
   /Game/Effects/Niagara/Modules/AnimTrail/FN_UpdateAnimTrail/FN_UpdateAnimTrail_v3/FN_UpdateAnimTrail()
   &FN_RibbonMaterial(
    Color=<rgba 1 1 1 0.4>
    Color Pan U Speed=1
    Color Scale U=0.5
    Erosion Amount=0.25
    Macro Distortion Amount=1
    Macro Pan Offset=1
    Macro Pan Speed=1
    Micro Distortion Amount=15
    Opacity Pan U Offset=1
    Opacity Pan U Speed=0.75
    Opacity Scale U=0.5
    Scale Alpha:NiagaraFloat=&FloatFromCurve(FloatCurve=DataInterface() {propertyValues {Curve {keys [
         {arriveTangent=-2.8813, interpMode=RCIM_Cubic, leaveTangent=-2.8813, tangentMode=RCTM_User, time=0.1, value=0.95}
         {interpMode=RCIM_Cubic, tangentMode=RCTM_User, time=1, value=0}
        ]}}})
   )
  ]
  rendererClasses [&NiagaraRibbonRendererProperties]
  renderers [{rendererClass=&NiagaraRibbonRendererProperties}]
 }
}
summary {
 systemName=NS_Skyfire_LowGravity_Trails
 emitters [
  {emitterName=Decals_Loop, rendererClasses [&NiagaraSpriteRendererProperties]}
  {emitterName=FN_AnimTrail_with_MasterMaterial, rendererClasses [&NiagaraRibbonRendererProperties]}
 ]
 userVariables [
  User.bIsFrontEnd {type=NiagaraBool}
  User.bIsFrontEndPreview {type=NiagaraBool}
  User.Color {type=LinearColor, description="", defaultValue {struct=LinearColor, value=<rgba 0.14 0.3719 0.7 1>}}
  User.FEVec {type=Vector3f, description="", defaultValue {struct=Vector3f, value=<0 0 0>}}
  User.Mat {
   type=MaterialInterface
   description=""
   defaultValue {struct=NiagaraExt_VariableValue_Object, value {object=null, objectClass=/Script/Engine.MaterialInterface}}
  }
  User.RibbonIDOffset {type=NiagaraInt32}
  User.SkeletalMesh {
   type=NiagaraDataInterfaceSkeletalMesh
   defaultValue {
    struct=NiagaraExt_VariableValue_DataInterface
    value {dataInterface=&self:NiagaraDataInterfaceSkeletalMesh_0}
   }
  }
 ]
}
